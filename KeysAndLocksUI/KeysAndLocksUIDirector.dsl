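; KeysAndLocksUIDirector is a Dynamic UI Director that shows or hides the keys UI scene based on how many key items the local player carries.
; On BeginPlay it binds OnWorldItemsChanged to the Fort inventory context's world-items-changed event.
; The handler reads the key item quantity from the owning Athena player controller and caches it in CurrentKeyCount.
; When that count changes, it adds KeysScene if above zero and removes it otherwise.
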(event EventBeginPlay
  (Inventory|BindEventtoOnWorldItemsChanged (BlueprintContext|GetFortInventoryContext) (EventDispatchers|CreateEvent :!event OnWorldItemsChanged)))

(event EventTick
  (params (DeltaSeconds :type Float%20%28single-precision%29)))

(fn OnWorldItemsChanged
  (bind _asathena_player_controller (Utilities|Casting|CastToAthena_PlayerController (DynamicUIDirector|GetOwningLocalPlayerController)))
  (bind _output_get (LocalVariables|SetNewKeyCount (Inventory|GetItemQuantityOnPlayer _asathena_player_controller (LocalVariables|GetKeyItemDefinition))))
  (if (!= _output_get (Variables|Default|GetCurrentKeyCount))
    (bind _output_get_2 (Variables|Default|SetCurrentKeyCount (LocalVariables|GetNewKeyCount)))
    (if (> _output_get_2 0)
      (DynamicUIDirector|AddScene "/KeysAndLocksUI/UI/KeysScene.KeysScene")
      (else
        (DynamicUIDirector|RemoveScene "/KeysAndLocksUI/UI/KeysScene.KeysScene")))))

(fn ConstructionScript)
